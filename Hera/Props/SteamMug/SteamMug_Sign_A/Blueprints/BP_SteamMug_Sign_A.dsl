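; Steam Mug sign prop Blueprint that drives a timed material effect when damaged.
; On damage it fires a cosmetic event with the current reset flag, which tells a first hit from a repeat.
; Retriggerable delays, using in-out duration for first hits and out duration for repeats, then clear the flag.
; Where cosmetics can run, game time and the flag are packed into the mesh's custom primitive data for the material.
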(event EventBeginPlay
  (EventDispatchers|BindEventtoOnDamaged (delegate Custom|OnDamaged_Event)))

(event Custom|OnDamaged_Event
  (params (DamagedActor :type Actor%20Object%20Reference) (Damage :type Float%20%28single-precision%29) (InstigatedBy :type Controller%20Object%20Reference) (DamageCauser :type Actor%20Object%20Reference) (HitLocation :type Vector) (FHitComponent :type Primitive%20Component%20Object%20Reference) (BoneName :type Name) (Momentum :type Vector))
  (CallFunction|OnDamagedCosmeticEvent (Variables|Default|GetResetFlag))
  (if (Variables|Default|GetResetFlag)
    (Utilities|FlowControl|RetriggerableDelay (Variables|Default|GetEffectOutDuration))
    (else
      (Variables|Default|SetResetFlag true)
      (Utilities|FlowControl|RetriggerableDelay (Variables|Default|GetEffectInOutDuration))))
  (Variables|Default|SetResetFlag false))

(event Custom|OnDamagedCosmetic_Event
  (params (Reset :type Boolean))
  (Networking|CanExecuteCosmeticEvents
    (:True
      (Rendering|Material|SetCustomPrimitiveDataVector2 (Variables|Debug|GetStaticMeshComponent) 2 (Math|Vector2D|MakeVector2D (Utilities|Time|GetGameTimeinSeconds) (Math|Conversions|ToFloat(Boolean) Reset))))))

(fn ConstructionScript
  (|CallParentFunction))
