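# Prototype Default Lit material for the SteamMug sign that tests timestamp-driven animation timelines.
# It compares Time to a start timestamp in custom primitive data to build an in-then-out ramp, smoothed by a smoothstep function.
# A second data slot derives a ResetTimer mask, and a simpler timeline runs from a GameTime parameter.
# Base Color shows only a numeric DebugScalarValues readout on rotated, scaled UVs.
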Material /Hera/Props/SteamMug/SteamMug_Sign_A/Materials/M_PROTO_Timestamping
graph {
 properties {
  bAutomaticallySetUsageInEditor=false
  bUseMaterialAttributes=true
  previewMesh=/Hera/Props/SteamMug/SteamMug_Sign_A/Meshes/SM_SteamMug_Sign_A
  usedShadingModels=Default Lit
 }
 nodes [
  n1 = ScalarParameter() {bUseCustomPrimitiveData=true, parameterName=CPD_Timestamp_Start, primitiveDataIndex=2}
  n2 = Divide(A=Subtract(A=Time(), B=@n1), B=Multiply(A=ScalarParameter() {defaultValue=2, parameterName=Timestamp_DurationIn}, B=NamedRerouteUsage()))
  n3 = Time()
  n4 = ScalarParameter() {bUseCustomPrimitiveData=true, parameterName=CPD_Timestamp_Start, primitiveDataIndex=2}
  n5 = Add(A=@n4, B=Multiply(A=ScalarParameter() {defaultValue=2, parameterName=Timestamp_DurationIn}, B=NamedRerouteUsage()))
  n6 = Saturate(None=Divide(A=Subtract(A=@n3, B=@n5), B=ScalarParameter() {defaultValue=2, parameterName=Timestamp_DurationOut}))
  n7 = Saturate(None=@n4)
  n8 = Multiply(A=Multiply(A=Saturate(None=@n2), B=Saturate(None=@n1)), B=OneMinus(None=Multiply(A=@n6, B=@n7)))
  NamedRerouteDeclaration(None=MaterialFunctionCall!(In=NamedRerouteDeclaration(None=@n8) {name=Timeline_InOut, nodeColor=<rgba 0.5234 1 0 1>}) {materialFunction=/Game/Effects/Fort_Effects/Materials/_Functions/MF_Smoothstep_01}.Result) {name=Timeline_InOutSmooth, nodeColor=<rgba 1 0 0.8437 1>}
  n9 = Saturate(None=Ceil(None=ScalarParameter() {bUseCustomPrimitiveData=true, parameterName=CPD_Timestamp_Reset, primitiveDataIndex=3}))
  NamedRerouteDeclaration(None=Clamp(None=OneMinus(None=@n9)) {minDefault=0}) {name=ResetTimer, nodeColor=<rgba 1 0.3516 0 1>}
  n10 = ScalarParameter() {parameterName=GameTime}
  n11 = Divide(A=Subtract(A=Time(), B=@n10), B=ScalarParameter() {defaultValue=5, parameterName=Duration})
  NamedRerouteDeclaration(None=Multiply(A=Saturate(None=@n11), B=Saturate(None=@n10))) {name=Timeline, nodeColor=<rgba 0.5703 1 0 1>}
  n12 = Add(
   A=MaterialFunctionCall(UVs=TextureCoordinate(), "Rotation Angle (0-1)"=Constant() {r=0.25}) {materialFunction=&CustomRotator}.Rotated Values
   B=Constant2Vector() {g=0.25, r=0.1}
  )
  n13 = Constant2Vector() {g=0.3, r=0.2}
  SetMaterialAttributes!(
   MaterialAttributes
   Base Color=MaterialFunctionCall(
    Number=NamedRerouteUsage()
    MaximumNumberOfDigits=Constant() {r=3}
    UVs=MaterialFunctionCall(UVs=@n12, Texture Scale=@n13) {materialFunction=&ScaleUVsByCenter}.UVs
   ) {materialFunction=&DebugScalarValues}.Result
  ) {attributeSetTypes [#1]}
 ]
}
